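; Base animation notify for Caretaker monster footsteps, handling the step's gameplay cue and stomp damage trigger.
; With a valid mesh owner, it casts to the Caretaker pawn and locally executes the sideways or normal footstep cue based on the owner's sideways tag.
; A failed cast uses the normal cue with no target actor.
; On dedicated servers, it sends the owner a gameplay event with the stomp damage ability tag, if valid.
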(fn Received_Notify
  (params (MeshComp :type Skeletal%20Mesh%20Component%20Object%20Reference) (Animation :type Anim%20Sequence%20Base%20Object%20Reference) (EventReference :type Anim%20Notify%20Event%20Reference%20Structure%20%28by%20ref%29))
  (bind _returnvalue (Components|GetOwner MeshComp))
  (bind _returnvalue_2 (Ability|GameplayCue|MakeGameplayCueParameters 0.0))
  (bind _stompdamageabilitytag (Variables|Default|GetStompDamageAbilityTag))
  ;[0 1 0.8739] Play the footstep gameplay cue based on if the actor is in the sideways or not.
  (Utilities|IsValid _returnvalue
    (:Is%20Valid
      (Utilities|Casting|CastToNPC_Pawn_Irwin_Monster_Caretaker _returnvalue
        (:then
          (bind _returnvalue_3 (GameplayTags|HasAnyMatchingGameplayTags(Message) _returnvalue (Variables|Default|GetSidewaysGameplayTag)))
          (GameplayCue|ExecuteGameplayCue%28Burst%29Local _returnvalue (select _returnvalue_3 (Variables|Default|GetGC_FootStep_Sideways) (Variables|Default|GetGC_FootStep)) _returnvalue_2))
        (:CastFailed
          (GameplayCue|ExecuteGameplayCue%28Burst%29Local 0 (Variables|Default|GetGC_FootStep) _returnvalue_2)))
      (if (Networking|IsDedicatedServer)
        (if (GameplayTags|IsGameplayTagValid _stompdamageabilitytag)
          (Ability|SendGameplayEventtoActor _returnvalue _stompdamageabilitytag (Utilities|Struct|MakeGameplayEventData _stompdamageabilitytag))
          (else
            _))
        (return true)))))
