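; Anim notify for the Caretaker's left footstep, marking the step time on the mesh's materials.
; When fired, it writes the current game time in seconds to the GameTimeFromBP_Left scalar parameter on all materials of the skeletal mesh.
; It then calls the parent notify implementation and returns its result.
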(fn Received_Notify
  (params (MeshComp :type Skeletal%20Mesh%20Component%20Object%20Reference) (Animation :type Anim%20Sequence%20Base%20Object%20Reference) (EventReference :type Anim%20Notify%20Event%20Reference%20Structure%20%28by%20ref%29))
  (Rendering|Material|SetScalarParameterValueonMaterials MeshComp "GameTimeFromBP_Left" (Utilities|Time|GetGameTimeinSeconds))
  (bind _returnvalue (|CallParentFunction MeshComp Animation))
  (return _returnvalue))
