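; An animation notify state that temporarily reshapes the Caretaker monster's weapon capsule during a span of an animation.
; On begin, it casts the mesh owner to the Caretaker pawn and requests a capsule change using the notify's half height, offset and radius variables.
; On end, it asks the same pawn to stop the capsule change.
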(fn Received_NotifyBegin
  (params (MeshComp :type Skeletal%20Mesh%20Component%20Object%20Reference) (Animation :type Anim%20Sequence%20Base%20Object%20Reference) (TotalDuration :type Float%20%28single-precision%29) (EventReference :type Anim%20Notify%20Event%20Reference%20Structure%20%28by%20ref%29))
  (bind _asnpc_pawn_irwin_monster_caretaker (Utilities|Casting|CastToNPC_Pawn_Irwin_Monster_Caretaker (Components|GetOwner MeshComp)))
  (Class|NPCPawnIrwinMonsterCaretaker|RequestWeaponCapsuleChange _asnpc_pawn_irwin_monster_caretaker (Variables|Default|GetWeaponCapsuleHalfHeight) (Variables|Default|GetWeaponCapsuleOffset) (Variables|Default|GetWeaponCapsuleRadius))
  (return false))

(fn Received_NotifyEnd
  (params (MeshComp :type Skeletal%20Mesh%20Component%20Object%20Reference) (Animation :type Anim%20Sequence%20Base%20Object%20Reference) (EventReference :type Anim%20Notify%20Event%20Reference%20Structure%20%28by%20ref%29))
  (bind _asnpc_pawn_irwin_monster_caretaker (Utilities|Casting|CastToNPC_Pawn_Irwin_Monster_Caretaker (Components|GetOwner MeshComp)))
  (Class|NPCPawnIrwinMonsterCaretaker|RequestStopWeaponCapsuleChange _asnpc_pawn_irwin_monster_caretaker)
  (return false))
